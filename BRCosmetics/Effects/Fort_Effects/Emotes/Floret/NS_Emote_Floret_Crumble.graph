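# Niagara system for the Floret emote's crumble effect, dropping small tinted debris from roughly head height.
# Rocks combines two bursts with a RockProbability-gated, two-pulse spawn curve, producing tumbling, randomly chosen meshes that fall and shrink.
# Dirt001 bursts near the origin without active motion, ramping a dynamic material parameter over its lifetime.
# The Dirt, Dust and Sand emitters are disabled, leaving the DustProbability input without effect.
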NiagaraSystem /BRCosmetics/Effects/Fort_Effects/Emotes/Floret/NS_Emote_Floret_Crumble
dependencies {
 usedDataInterfaces [
  &NiagaraDataInterfaceCurve
  &NiagaraDataInterfaceMeshRendererInfo
  &NiagaraDataInterfaceVector2DCurve
 ]
 usedDynamicInputs [&RemapRange]
 usedModules [
  &self:SystemSpawnScript.SystemScriptSource.SystemScriptGraph.NiagaraNodeAssignment_0.SetVariables_#1
  &self:SystemSpawnScript.SystemScriptSource.SystemScriptGraph.NiagaraNodeAssignment_1.SetVariables_#2
  &self:SystemSpawnScript.SystemScriptSource.SystemScriptGraph.NiagaraNodeAssignment_2.SetVariables_#3
  &SystemState
 ]
}
emitters {
 Dirt {
  bEnabled=false
  simTarget=GPUComputeSim
  emitterUpdateScript [
   &EmitterState()
   &SpawnBurst_Instantaneous(Spawn Time=0.1, Use Spawn Probability=false, UseLoopCountLimit=false) {module=SpawnBurst_Instantaneous001}
   &SpawnBurst_Instantaneous(Spawn Time=0.08, Use Spawn Probability=false, UseLoopCountLimit=false)
   &SpawnRate(
    Spawn Probability=$User.DustProbability
    SpawnRate=&FloatFromCurve(
     FloatCurve=DataInterface() {propertyValues {Curve {keys [
         {interpMode=RCIM_Cubic, tangentMode=RCTM_Break, time=0, value=0.2}
         {interpMode=RCIM_Cubic, tangentMode=RCTM_Break, time=0.3, value=1}
         {arriveTangent=-3.0515, interpMode=RCIM_Cubic, tangentMode=RCTM_Break, time=1, value=0.1}
        ]}}}
     CurveIndex=$Emitter.NormalizedLoopAge
     Scale Curve=&RandomRangeFloat(Minimum=2, Maximum=4)
    )
    Use Spawn Probability=true
   ) {enabled=false}
  ]
  particleSpawnScript [
   &InitializeParticle(
    Color=<rgba 0.402 0.0212 0.0666 1>
    Lifetime=0.8
    Position Offset=<0 8 175>
    Sprite Rotation Mode=::NewEnumerator2
    Sprite Size=&RandomRangeVector2D(Minimum=<50 50>, Maximum=<50 50>)
    Sprite Size Mode=::NewEnumerator4
    UsePositionOffset=true
    Write Material Random=true
   )
   &SubUVAnimation(End Frame=3, SubUV Animation Mode:ENiagara_SubUVLookupMode=::NewEnumerator2)
   &ShapeLocation(Box Size=<3 3 3>, MaskSpawnGroup=false, Override Seed=false, Shape Primitive=::NewEnumerator2, Surface Only=false, UseOwnerScale=true)
  ]
  particleUpdateScript [
   &ParticleState()
   &GravityForce(Gravity=&RandomRangeVector(Minimum=<0 0 -300>, Maximum=<0 0 -300>))
   &Drag(Drag=&RandomRangeFloat(Minimum=2, Maximum=2.2), Use Linear Drag=true, Use Rotational Drag=false)
   &SolveForcesAndVelocity(Clamp Velocity=false, Limit Acceleration=false)
   &ScaleSpriteSize(Uniform Curve Sprite Scale=DataInterface() {propertyValues {Curve {keys [{interpMode=RCIM_Cubic, tangentMode=RCTM_User, time=0, value=0.3}, {interpMode=RCIM_Cubic, tangentMode=RCTM_User, time=1, value=1}]}}})
   &ScaleColor(
    Scale Alpha=&FloatFromCurve(FloatCurve=DataInterface() {propertyValues {Curve {keys [
         {arriveTangent=-0.4559, interpMode=RCIM_Cubic, leaveTangent=-0.4559, tangentMode=RCTM_User, time=0, value=1}
         {arriveTangent=-2.4378, interpMode=RCIM_Cubic, leaveTangent=-2.4378, tangentMode=RCTM_User, time=1.0011, value=0.002}
        ]}}})
    ScaleA=true
    ScaleRGB=true
   )
  ]
  rendererClasses [&NiagaraSpriteRendererProperties]
  renderers [{rendererClass=&NiagaraSpriteRendererProperties}]
 }
 Dirt001 {
  emitterUpdateScript [
   &EmitterState()
   &SpawnBurst_Instantaneous(Use Spawn Probability=false, UseLoopCountLimit=false)
  ]
  particleSpawnScript [
   &InitializeParticle(
    Color=<rgba 0.402 0.0212 0.0666 1>
    Lifetime=1.5
    Mesh Scale Mode=::NewEnumerator3
    Position Offset=<0 0 9>
    Sprite Rotation Mode=::NewEnumerator2
    UsePositionOffset=true
    Write Material Random=true
   )
   &SubUVAnimation(End Frame=3, SubUV Animation Mode:ENiagara_SubUVLookupMode=::NewEnumerator2) {enabled=false}
   &SphereLocation(
    MaskSpawnGroup=false
    Override Seed=false
    Sphere Origin:NiagaraPosition=&Add_Vector(A=Linked() {linkedVariable {name=Particles.Position, type=NiagaraPosition}}, B=$System.RockDustPos)
    Sphere Radius=30
    Surface Only=false
    UseNonUniformScale=true
    UseOffset=false
    UseSphereOrientation=false
   ) {enabled=false}
  ]
  particleUpdateScript [
   &ParticleState()
   &GravityForce(Gravity=&RandomRangeVector(Minimum=<0 0 -1000>, Maximum=<0 0 -1200>)) {enabled=false}
   &Drag(Drag=&RandomRangeFloat(Minimum=0.5, Maximum=1.5), Use Linear Drag=true, Use Rotational Drag=false) {enabled=false}
   &SolveForcesAndVelocity(Clamp Velocity=false, Limit Acceleration=false) {enabled=false}
   &ScaleSpriteSize(Uniform Curve Sprite Scale=DataInterface() {propertyValues {Curve {keys [
        {arriveTangent=0.453, interpMode=RCIM_Cubic, leaveTangent=0.453, tangentMode=RCTM_User, time=0, value=0.2}
        {arriveTangentWeight=0.1414, interpMode=RCIM_Cubic, leaveTangentWeight=0.3299, tangentMode=RCTM_User, time=0.4, value=1}
        {arriveTangent=-0.0405, interpMode=RCIM_Cubic, leaveTangent=-0.0405, tangentMode=RCTM_User, time=1, value=0.3}
       ]}}}) {enabled=false}
   &ScaleSpriteSize(Scale Sprite Size Mode=::NewEnumerator0, Uniform Scale Factor=&ClampFloat(Float=$User.ParticleScale, Min=0.4)) {enabled=false, module=ScaleSpriteSize001}
   &ScaleColor(
    Scale Alpha=&FloatFromCurve(FloatCurve=DataInterface() {propertyValues {Curve {keys [
         {arriveTangent=-0.4559, interpMode=RCIM_Cubic, leaveTangent=-0.4559, tangentMode=RCTM_User, time=0, value=1}
         {arriveTangent=-2.4378, interpMode=RCIM_Cubic, leaveTangent=-2.4378, tangentMode=RCTM_User, time=1.0011, value=0.002}
        ]}}})
    Scale RGB=&VectorFromFloat(Value=&RandomRangeFloat(Minimum=0, Maximum=1))
    ScaleA=true
    ScaleRGB=true
   ) {enabled=false}
   &DynamicMaterialParameters(
    Index 0 Param 1=&Lerp_Float(
     A=1
     B=64
     Alpha=&FloatFromCurve(FloatCurve=DataInterface() {propertyValues {Curve {keys [
          {arriveTangent=0.441, interpMode=RCIM_Cubic, leaveTangent=0.441, tangentMode=RCTM_User, time=0, value=0}
          {arriveTangent=0.9102, interpMode=RCIM_Cubic, leaveTangent=0.9102, tangentMode=RCTM_User, time=1, value=1}
         ]}}})
    )
    Index 0 Param 2=165
    Index 0 Param 3=40
    Module.Index0Param0Write=true
    Module.Index0Param1Write=true
    Module.Index0Param2Write=true
    Module.Index0Param3Write=true
   )
  ]
  rendererClasses [&NiagaraMeshRendererProperties, &NiagaraSpriteRendererProperties]
  renderers [
   {rendererClass=&NiagaraMeshRendererProperties, rendererIndex=0}
   {rendererClass=&NiagaraSpriteRendererProperties, rendererIndex=1}
  ]
 }
 Dust {
  bEnabled=false
  emitterUpdateScript [
   &EmitterState()
   &SpawnBurst_Instantaneous(Spawn Count=2, Use Spawn Probability=false, UseLoopCountLimit=false)
   &SpawnRate(
    SpawnRate=&Lerp_Float(
     A=3
     B=10
     Alpha=&FloatFromCurve(
      FloatCurve=DataInterface() {propertyValues {Curve {keys [
          {arriveTangent=0.4438, interpMode=RCIM_Cubic, leaveTangent=0.4438, tangentMode=RCTM_User, time=0, value=1}
          {arriveTangent=-8.446, tangentMode=RCTM_User, time=0.2, value=0}
          <at 0.6 0>
          <at 0.7 1 RCIM_Cubic>
          {arriveTangent=-4.5805, interpMode=RCIM_Cubic, leaveTangent=-4.5805, tangentMode=RCTM_User, time=1, value=0}
         ]}}}
      CurveIndex=$Emitter.NormalizedLoopAge
     )
    )
    Use Spawn Probability=true
   )
  ]
  particleSpawnScript [
   &InitializeParticle(
    Color=<rgba 0.402 0.0212 0.0666 1>
    Lifetime=&RandomRangeFloat(Minimum=1.5, Maximum=2)
    Position Offset=<0 8 185>
    Sprite Rotation Mode=::NewEnumerator2
    Sprite Size=&RandomRangeVector2D(Minimum=<35 45>, Maximum=<40 50>)
    Sprite Size Mode=::NewEnumerator4
    UsePositionOffset=true
    Write Material Random=true
   )
   &ShapeLocation(Box Size=<20 5 5>, MaskSpawnGroup=false, Override Seed=false, Shape Primitive=::NewEnumerator2, Surface Only=false, UseOwnerScale=true)
  ]
  particleUpdateScript [
   &ParticleState()
   &Drag(
    Drag=&Multiply_Float(
     A=&RandomRangeFloat(Minimum=6, Maximum=8)
     B=&FloatFromCurve(FloatCurve=DataInterface() {propertyValues {Curve {keys [
          {interpMode=RCIM_Cubic, leaveTangent=2.8725, tangentMode=RCTM_Break, time=0, value=0}
          {interpMode=RCIM_Cubic, tangentMode=RCTM_Break, time=1, value=1}
         ]}}})
    )
    Use Linear Drag=true
    Use Rotational Drag=false
   )
   &SubUVAnimation()
   &GravityForce(Gravity=&RandomRangeVector(Minimum=<0 0 -175>, Maximum=<0 0 -250>))
   &SolveForcesAndVelocity(Clamp Velocity=false, Limit Acceleration=false)
   &ScaleColor(
    Scale Alpha=&FloatFromCurve(
     FloatCurve=DataInterface() {propertyValues {Curve {keys [
         {arriveTangent=2.4826, interpMode=RCIM_Cubic, leaveTangent=2.4826, tangentMode=RCTM_User, time=0, value=0}
         {arriveTangentWeight=0.1842, interpMode=RCIM_Cubic, leaveTangentWeight=0.7369, tangentMode=RCTM_User, time=0.3, value=1}
         <at 1 0>
        ]}}}
     Scale Curve=&RandomRangeFloat(Minimum=0.15, Maximum=0.35)
    )
    ScaleA=true
    ScaleRGB=true
   )
   &DynamicMaterialParameters(
    Index 0 Param 1=&FloatFromCurve(FloatCurve=DataInterface() {propertyValues {Curve {keys [
         {interpMode=RCIM_Cubic, leaveTangent=-0.0039, tangentMode=RCTM_Break, time=0, value=0}
         {arriveTangent=-1.8971, interpMode=RCIM_Cubic, tangentMode=RCTM_Break, time=1, value=-1}
        ]}}})
    Index 0 Param 2=1
    Index 0 Param 3=0.5
    Index 0 Param 4=0.5
    Module.Param0WriteEnabled=true
    Module.Param1WriteEnabled=true
    Module.Param2WriteEnabled=true
    Module.Param3WriteEnabled=true
   )
   &ScaleSpriteSize(
    Uniform Curve Scale=1.5
    Uniform Curve Sprite Scale=DataInterface() {propertyValues {Curve {keys [
        {interpMode=RCIM_Cubic, leaveTangent=0.5294, tangentMode=RCTM_Break, time=0, value=0.5}
        {arriveTangent=3.4734, interpMode=RCIM_Cubic, tangentMode=RCTM_Break, time=1, value=2}
       ]}}}
   )
   &SpriteRotationRate(Rotation Rate=&Multiply_Float(
     A=&Multiply_Float(
      A=&Lerp_Float(
       A=10
       B=30
       Alpha=&FloatFromCurve(
        FloatCurve=DataInterface() {propertyValues {Curve {keys [
            {interpMode=RCIM_Cubic, tangentMode=RCTM_Break, time=0, value=0}
            {arriveTangent=2.8858, interpMode=RCIM_Cubic, tangentMode=RCTM_Break, time=1, value=1}
           ]}}}
        CurveIndex=&RandomRangeFloat(Minimum=0, Maximum=1)
       )
      )
      B=&UniformAOrBFloat(A=-1)
     )
     B=&FloatFromCurve(FloatCurve=DataInterface() {propertyValues {Curve {keys [
          {interpMode=RCIM_Cubic, tangentMode=RCTM_Break, time=0.1, value=0}
          {arriveTangent=2.6363, interpMode=RCIM_Cubic, tangentMode=RCTM_Break, time=1, value=1}
         ]}}})
    )) {enabled=false}
  ]
  rendererClasses [&NiagaraSpriteRendererProperties]
  renderers [{rendererClass=&NiagaraSpriteRendererProperties}]
 }
 Rocks {
  emitterUpdateScript [
   &EmitterState()
   &SpawnBurst_Instantaneous(Spawn Count=3, Spawn Time=0.5, Use Spawn Probability=false, UseLoopCountLimit=false) {module=SpawnBurst_Instantaneous001}
   &SpawnBurst_Instantaneous(Spawn Count=3, Spawn Time=0.05, Use Spawn Probability=false, UseLoopCountLimit=false)
   &SpawnRate(
    Spawn Probability=$User.RockProbability
    SpawnRate=&Lerp_Float(
     A=6
     B=13
     Alpha=&FloatFromCurve(
      FloatCurve=DataInterface() {propertyValues {Curve {keys [
          {arriveTangent=0.4438, interpMode=RCIM_Cubic, leaveTangent=0.4438, tangentMode=RCTM_User, time=0, value=1}
          {arriveTangent=-14.218, tangentMode=RCTM_User, time=0.2, value=0}
          <at 0.6 0>
          <at 0.7 1 RCIM_Cubic>
          {arriveTangent=-10.0648, interpMode=RCIM_Cubic, leaveTangent=-10.0648, tangentMode=RCTM_User, time=1, value=0}
         ]}}}
      CurveIndex=$Emitter.NormalizedLoopAge
     )
    )
    Use Spawn Probability=true
   )
  ]
  particleSpawnScript [
   &InitializeParticle(
    Color=<rgba 0.402 0.0212 0.0666 1>
    Lifetime=&RandomRangeFloat(Minimum=0.5, Maximum=0.6)
    Mesh Scale=&Multiply_VectorByFloat(Vector=&RandomRangeVector(Minimum=<0.7 0.7 0.7>, Maximum=<1 1 1>), Float=0.04)
    Mesh Scale Mode=::NewEnumerator4
    Position Offset=<0 8 184>
    UsePositionOffset=true
    Write Material Random=true
   )
   &ShapeLocation(Box Size=<30 5 5>, MaskSpawnGroup=false, Override Seed=false, Shape Primitive=::NewEnumerator2, Surface Only=false, UseOwnerScale=true)
   &InitialMeshOrientation(Use Rotation Vector=false)
   &AddRotationalVelocity(Rotation Rate=&Multiply_VectorByFloat(
     Vector=&RandomVector(Vector Scale=2)
     Float=&RandomRangeFloat(Minimum=0.5, Maximum=1)
    ))
   &DynamicMaterialParameters(
    Index 0 Param 2=&RandomRangeFloat(Minimum=0.25, Maximum=1)
    Module.Param0WriteEnabled=false
    Module.Param1WriteEnabled=true
    Module.Param2WriteEnabled=false
    Module.Param3WriteEnabled=false
   )
   /SetVariables(Particles.MeshIndex=&RandomRangeInt(Minimum=0, Maximum=2)) {isSetParameters=true}
  ]
  particleUpdateScript [
   &ParticleState()
   &GravityForce(Gravity=&RandomRangeVector(Minimum=<0 0 -500>, Maximum=<0 0 -550>))
   &Collision(
    Age Colliding Particles=false
    Module.Control Rest State Via Surface Normal=false
    Module.EnableMaxCollisionCount=false
    Module.Randomize Collision Normal Vector=false
    Restitution=0.1
   ) {enabled=false}
   &SolveForcesAndVelocity(Clamp Velocity=false, Limit Acceleration=false)
   &ScaleMeshSize(Scale Factor=&VectorFromFloat(Value=&FloatFromCurve(
      FloatCurve=DataInterface() {propertyValues {Curve {keys [{arriveTangent=-0.0035, interpMode=RCIM_Cubic, leaveTangent=-0.0035, tangentMode=RCTM_User, time=0.4, value=1}, <at 1 0>]}}}
      Scale Curve=&RandomRangeFloat(Minimum=0.6, Maximum=1.2)
     )))
  ]
  rendererClasses [&NiagaraMeshRendererProperties, &NiagaraSpriteRendererProperties]
  renderers [
   {rendererClass=&NiagaraMeshRendererProperties, rendererIndex=0}
   {rendererClass=&NiagaraSpriteRendererProperties, rendererIndex=1}
  ]
 }
 Sand {
  bEnabled=false
  emitterUpdateScript [
   &EmitterState()
   &SpawnRate(
    Spawn Probability=$User.DustProbability
    SpawnRate=&FloatFromCurve(
     FloatCurve=DataInterface() {propertyValues {Curve {keys [
         {interpMode=RCIM_Cubic, tangentMode=RCTM_Break, time=0, value=0.2}
         {interpMode=RCIM_Cubic, tangentMode=RCTM_Break, time=0.3, value=1}
         {arriveTangent=-3.0515, interpMode=RCIM_Cubic, tangentMode=RCTM_Break, time=1, value=0.1}
        ]}}}
     CurveIndex=$Emitter.NormalizedLoopAge
     Scale Curve=22
    )
    Use Spawn Probability=true
   )
   /SetVariables(Emitter.DepthFade=10) {isSetParameters=true}
  ]
  particleSpawnScript [
   /SetVariables(Particles.RandomFloat=&UniformRangedFloat(Minimum=0, Maximum=3)) {enabled=false, isSetParameters=true}
   &InitializeParticle(
    Color=<rgba 0.8963 0.0356 0.0823 1>
    Position Offset=<0 8 180>
    Sprite Rotation Mode=::NewEnumerator2
    Sprite Size Max=<20 30>
    Sprite Size Min=<20 30>
    Sprite Size Mode=::NewEnumerator2
    UsePositionOffset=true
    Write Material Random=true
   )
   &SubUVAnimation(End Frame=3, SubUV Animation Mode:ENiagara_SubUVLookupMode=::NewEnumerator2)
   &ShapeLocation(Box Size=<5 3 3>, MaskSpawnGroup=false, Override Seed=false, Shape Primitive=::NewEnumerator2, Surface Only=false, UseOwnerScale=true)
  ]
  particleUpdateScript [
   &ParticleState()
   &GravityForce(Gravity=&RandomRangeVector(Minimum=<0 0 -800>, Maximum=<0 0 -500>))
   &SolveForcesAndVelocity(Clamp Velocity=false, Limit Acceleration=false)
   &ScaleSpriteSize(Scale Factor=&Vector2DFromCurve(
     Vector2Curve=DataInterface() {propertyValues {XCurve {keys [<at 0.0036 1.6897>,<at 0.9967 0.0582>]}, YCurve {keys [<at 0 1>,<at 1.002 3.7227>]}}}
     ScaleCurve=<1 1>
    ))
   &ScaleColor(
    Scale Alpha=&FloatFromCurve(
     FloatCurve=DataInterface() {propertyValues {Curve {keys [
         {interpMode=RCIM_Cubic, leaveTangentWeight=0.0962, tangentMode=RCTM_Break, time=0, value=0}
         {interpMode=RCIM_Cubic, tangentMode=RCTM_Break, time=0.2, value=1}
         {interpMode=RCIM_Cubic, tangentMode=RCTM_Break, time=1, value=0}
        ]}}}
     Scale Curve=&RandomRangeFloat(Minimum=0.05, Maximum=0.15)
    )
    ScaleA=true
    ScaleRGB=false
   )
   &DynamicMaterialParameters(
    Index 0 Param 2=1
    Index 0 Param 3=1
    Index 0 Param 4=&FloatFromCurve(FloatCurve=DataInterface() {propertyValues {Curve {keys [{interpMode=RCIM_Cubic, tangentMode=RCTM_Break, time=0.3, value=1}, {interpMode=RCIM_Cubic, tangentMode=RCTM_Break, time=1, value=0.25}]}}})
    Module.Param0WriteEnabled=true
    Module.Param1WriteEnabled=true
    Module.Param2WriteEnabled=true
    Module.Param3WriteEnabled=true
   )
  ]
  rendererClasses [&NiagaraSpriteRendererProperties]
  renderers [{rendererClass=&NiagaraSpriteRendererProperties}]
 }
}
summary {
 systemName=NS_Emote_Floret_Crumble
 emitters [
  {emitterName=Rocks, rendererClasses [&NiagaraMeshRendererProperties, &NiagaraSpriteRendererProperties]}
  {bEnabled=false, emitterName=Dirt, simTarget=GPUComputeSim, rendererClasses [&NiagaraSpriteRendererProperties]}
  {bEnabled=false, emitterName=Dust, rendererClasses [&NiagaraSpriteRendererProperties]}
  {bEnabled=false, emitterName=Sand, rendererClasses [&NiagaraSpriteRendererProperties]}
  {emitterName=Dirt001, rendererClasses [&NiagaraMeshRendererProperties, &NiagaraSpriteRendererProperties]}
 ]
 userVariables [
  User.AreaScale {type=NiagaraFloat, defaultValue=1, description=""}
  User.DustProbability {type=NiagaraFloat, defaultValue=1, description=""}
  User.ParticleScale {type=NiagaraFloat, defaultValue=0.5, description=""}
  User.RockProbability {type=NiagaraFloat, defaultValue=1, description=""}
  User.WaterProbability {type=NiagaraFloat, defaultValue=1, description=""}
 ]
}
