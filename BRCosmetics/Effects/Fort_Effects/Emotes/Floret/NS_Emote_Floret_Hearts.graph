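# Niagara effect for the Floret emote that releases pink hearts and a brief glow around upper-body height.
# LittleHearts mixes scattered small mesh hearts from a rate and burst spawn with a spawn-group-masked burst of larger, longer-lived, slowly rising ones.
# Drag, upward acceleration and curves shape their motion, pop-then-shrink scale and brightening.
# SimpleSpriteBurst adds a fading sprite plus pulsing pink light through component and light renderers, with component intensity scaled by time of day.
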NiagaraSystem /BRCosmetics/Effects/Fort_Effects/Emotes/Floret/NS_Emote_Floret_Hearts
dependencies {usedDataInterfaces [&NiagaraDataInterfaceCurve], usedDynamicInputs [], usedModules [&SystemState]}
emitters {
 LittleHearts {
  emitterUpdateScript [
   &EmitterState()
   &SpawnRate(SpawnRate=8, Use Spawn Probability=false)
   &SpawnBurst_Instantaneous(Spawn Count=2, Use Spawn Probability=false, UseLoopCountLimit=false)
   &SpawnBurst_Instantaneous(Spawn Group=1, Use Spawn Probability=false, UseLoopCountLimit=false) {module=SpawnBurst_Instantaneous001}
  ]
  particleSpawnScript [
   &InitializeParticle(
    Color=&RandomRangeLinearColor(Minimum=<rgba 3 0 0.1112 1>, Maximum=<rgba 3 0.0949 0.3092 1>)
    Lifetime=&MaskFloatBySpawnGroup(Masked Float=1.3, Passthrough Float=&RandomRangeFloat(Minimum=0.7, Maximum=1), Spawn Group Mask=1)
    Mesh Scale=&MaskVectorBySpawnGroup(
     Masked Vector=&VectorFromFloat(Value=&RandomRangeFloat(Minimum=0.3, Maximum=0.3))
     Passthrough Vector=&VectorFromFloat(Value=&RandomRangeFloat(Minimum=0.1, Maximum=0.16))
     Spawn Group Mask=1
    )
    Mesh Scale Mode=::NewEnumerator4
    Position Offset=<10 0 175>
    Position Offset Coordinate Space=::Local
    UsePositionOffset=true
    Write Material Random=true
   )
   &AddVelocity(Velocity=&MaskVectorBySpawnGroup(
     Masked Vector=<0 0 20>
     Passthrough Vector=&RandomRangeVector(Minimum=<-75 -75 -20>, Maximum=<75 75 80>)
     Spawn Group Mask=1
    ))
   &InitialMeshOrientation(
    Mesh Orientation Mode=::NewEnumerator6
    Orientation Quaternion:Quat4f=&MakeQuatFromAxisAngle(Axis=&RandomRangeVector(Minimum=<-0.05 -0.03 0>, Maximum=<0.05 0.03 0>), AngleInDegrees=90)
    Rotation=<0 0 0.25>
    Use Rotation Vector=true
   )
  ]
  particleUpdateScript [
   &ParticleState()
   &ScaleMeshSize(Scale Factor=&VectorFromFloat(Value=&FloatFromCurve(FloatCurve=DataInterface() {propertyValues {Curve {keys [
          {arriveTangent=3.6183, interpMode=RCIM_Cubic, time=0, value=0.2}
          {arriveTangent=1.1429, interpMode=RCIM_Cubic, leaveTangent=1.1429, time=0.4, value=1}
          {arriveTangent=0.75, interpMode=RCIM_Cubic, leaveTangent=0.75, time=0.7, value=1}
          {arriveTangent=-3.3333, interpMode=RCIM_Cubic, leaveTangent=-3.3333, time=0.8, value=1.3}
          {interpMode=RCIM_Cubic, leaveTangent=-33.019, time=1, value=0}
         ]}}})))
   &Drag(Drag=4, Use Linear Drag=true, Use Rotational Drag=true)
   &AccelerationForce(Acceleration=&RandomRangeVector(Minimum=<0 0 30>, Maximum=<0 0 80>))
   &SolveForcesAndVelocity(Clamp Velocity=false, Limit Acceleration=false)
   &ScaleColor(
    Scale RGB=&VectorFromFloat(Value=&FloatFromCurve(FloatCurve=DataInterface() {propertyValues {Curve {keys [
          {arriveTangent=4.1785, interpMode=RCIM_Cubic, leaveTangent=4.1785, tangentMode=RCTM_User, time=0, value=0.2}
          {arriveTangent=1.1179, interpMode=RCIM_Cubic, leaveTangent=1.1179, tangentMode=RCTM_User, time=0.2042, value=0.6228}
          {arriveTangent=0.0224, interpMode=RCIM_Cubic, leaveTangent=0.0224, tangentMode=RCTM_User, time=1, value=1}
         ]}}}))
    ScaleA=true
    ScaleRGB=true
   )
  ]
  rendererClasses [&NiagaraMeshRendererProperties]
  renderers [{rendererClass=&NiagaraMeshRendererProperties, rendererIndex=0}]
 }
 SimpleSpriteBurst {
  emitterUpdateScript [
   &EmitterState()
   &SpawnBurst_Instantaneous(Use Spawn Probability=false, UseLoopCountLimit=false)
  ]
  particleSpawnScript [&InitializeParticle(
    Lifetime=1.3
    Position Offset=<10 0 190>
    Sprite Size Mode=::NewEnumerator3
    Uniform Sprite Size=20
    UsePositionOffset=true
    Write Material Random=true
   )]
  particleUpdateScript [
   &ParticleState()
   &ScaleColor(
    Scale Alpha=&FloatFromCurve(FloatCurve=DataInterface() {propertyValues {Curve {keys [<at -0.0046 1.0113 RCIM_Cubic>,<at 1.0011 0.002 RCIM_Cubic>]}}})
    ScaleA=true
    ScaleRGB=false
   )
   &FN_Gen5_Light(
    Attenutation=75
    Color:LinearColor=<rgba 3 0.0949 0.3092 1>
    Exponent=2
    Intensity:NiagaraFloat=&Multiply_Float(
     A=&FNTimeOfDayScalar(Night Scale=150, Morning Scale=150, Day Scale=500, Evening Scale=150)
     B=&FloatFromCurve(FloatCurve=DataInterface() {propertyValues {Curve {keys [
          {arriveTangent=7.7337, interpMode=RCIM_Cubic, leaveTangent=7.7337, tangentMode=RCTM_User, time=0, value=0}
          {interpMode=RCIM_Cubic, tangentMode=RCTM_User, time=0.5, value=1}
          {arriveTangent=-6.8061, tangentMode=RCTM_User, time=1, value=0}
         ]}}})
    )
   )
   &FN_Light_Attributes(
    Color=<rgba 3 0.0949 0.3092 1>
    Create Color Attribute=true
    Create Radius Attribute=true
    Radius=50
    Scale Color=&FloatFromCurve(FloatCurve=DataInterface() {propertyValues {Curve {keys [
         {arriveTangent=7.7337, interpMode=RCIM_Cubic, leaveTangent=7.7337, tangentMode=RCTM_User, time=0, value=0}
         {interpMode=RCIM_Cubic, tangentMode=RCTM_User, time=0.5, value=1}
         {arriveTangent=-6.8061, tangentMode=RCTM_User, time=1, value=0}
        ]}}})
   )
  ]
  rendererClasses [
   &NiagaraSpriteRendererProperties
   &NiagaraComponentRendererProperties
   &NiagaraLightRendererProperties
  ]
  renderers [
   {rendererClass=&NiagaraSpriteRendererProperties}
   {rendererClass=&NiagaraComponentRendererProperties, rendererIndex=1}
   {rendererClass=&NiagaraLightRendererProperties, rendererIndex=2}
  ]
 }
}
summary {
 systemName=NS_Emote_Floret_Hearts
 emitters [
  {emitterName=LittleHearts, rendererClasses [&NiagaraMeshRendererProperties]}
  {
   emitterName=SimpleSpriteBurst
   rendererClasses [
    &NiagaraSpriteRendererProperties
    &NiagaraComponentRendererProperties
    &NiagaraLightRendererProperties
   ]
  }
 ]
 userVariables []
}
