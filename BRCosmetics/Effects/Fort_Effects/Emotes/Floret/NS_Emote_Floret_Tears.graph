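# Niagara system for the Floret emote that sprays short-lived mesh teardrops from a character's skeletal mesh.
# A single emitter spawns drips intermittently at a skeletal sample point, partitioning them into two streams with mirrored offsets and velocities, presumably one per eye.
# A large-body user flag switches the offsets to fit bigger pawns.
# Drag and gravity shape the fall while curves reshape and fade each drop.
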NiagaraSystem /BRCosmetics/Effects/Fort_Effects/Emotes/Floret/NS_Emote_Floret_Tears
dependencies {usedDataInterfaces [&NiagaraDataInterfaceCurve], usedDynamicInputs [], usedModules [&SystemState]}
emitters {Drips001 {
  emitterUpdateScript [
   &EmitterState(
    Life Cycle Mode=::NewEnumerator1
    Loop Duration=&RandomRangeFloat(Minimum=0.3, Maximum=0.5)
    UseLoopDelay=false
   )
   &SpawnRate(Spawn Probability=0.73, SpawnRate=5, Use Spawn Probability=true)
  ]
  particleSpawnScript [
   &InitializeParticle(
    Color Maximum=<rgba 1 1 1 6>
    Color Minimum=<rgba 1 1 1 3>
    Color Mode=::NewEnumerator2
    Lifetime Max=0.6
    Lifetime Min=0.35
    Lifetime Mode=::NewEnumerator1
    Mesh Scale Max=<0.6 0.6 0.6>
    Mesh Scale Min=<0.3 0.3 0.3>
    Mesh Scale Mode=::NewEnumerator2
    Position Mode=::NewEnumerator1
    UsePositionOffset=false
    Write Material Random=true
   ) {module=InitializeParticle001}
   &PartitionParticles()
   &SkeletalMeshLocation(
    Bone Sampling Mode=::NewEnumerator4
    MaskSpawnGroup=false
    Orientation Sampling=::NewEnumerator1
    Sampled Position Offset=&MakeCustomVectorFromBool(
     Bool=&SetBoolByIntComparison(A=$Particles.PartitionParticles.Partition, B=0, Comparison Type=::NewEnumerator4)
     True Vector=&MakeCustomVectorFromBool(Bool=$User.bIsLargeBodyPawn, True Vector=<5 5 -10>, False Vector=<5 8 -4>)
     False Vector=&MakeCustomVectorFromBool(Bool=$User.bIsLargeBodyPawn, True Vector=<5 5 10>, False Vector=<5 8 5>)
    )
    Sampled Position Offset Coordinate Space:ENiagaraSkelMeshTransforms=::NewEnumerator2
    Skeletal Mesh=$User.SkeletalMesh
   ) {module=SkeletalMeshLocation001}
   &ShapeLocation(MaskSpawnGroup=false, Override Seed=false, Sphere Radius=1, UseOwnerScale=true)
   &CameraOffset()
   &AddVelocity(
    Velocity=&MakeCustomVectorFromBool(
     Bool=&SetBoolByIntComparison(A=$Particles.PartitionParticles.Partition, B=0, Comparison Type=::NewEnumerator4)
     True Vector=&RandomRangeVector(Minimum=<-12 -24 -36>, Maximum=<36 24 -72>)
     False Vector=&RandomRangeVector(Minimum=<-12 -24 36>, Maximum=<36 24 72>)
    )
    Velocity Speed Scale=&RandomRangeFloat(Minimum=0.5, Maximum=0.6)
   )
  ]
  particleUpdateScript [
   &ParticleState()
   &DynamicMaterialParameters(
    Index 0 Param 1=&RandomRangeFloat(Minimum=0, Maximum=1)
    Index 0 Param 2=1
    Module.Param0WriteEnabled=true
    Module.Param1WriteEnabled=true
    Module.Param2WriteEnabled=true
    Module.Param3WriteEnabled=true
   )
   &Drag(Drag=&RandomRangeFloat(Minimum=2, Maximum=6), Use Linear Drag=true, Use Rotational Drag=false)
   &GravityForce(Gravity=&RandomRangeVector(Minimum=<0 0 -150>, Maximum=<0 0 -300>))
   &SolveForcesAndVelocity(Clamp Velocity=false, Limit Acceleration=false)
   &ScaleMeshSize(Scale Factor=&MakeVector(
     X=&FloatFromCurve(FloatCurve=DataInterface() {propertyValues {Curve {keys [<at 0 0.5 RCIM_Cubic>, {interpMode=RCIM_Cubic, leaveTangent=0.5, time=0.2, value=1}]}}})
     Y=&FloatFromCurve(FloatCurve=DataInterface() {propertyValues {Curve {keys [<at 0 1.5 RCIM_Cubic>, {arriveTangent=-1, interpMode=RCIM_Cubic, leaveTangent=-1, time=0.2, value=1}, <at 1 0.5>]}}})
     Z=&FloatFromCurve(FloatCurve=DataInterface() {propertyValues {Curve {keys [<at 0 1.5 RCIM_Cubic>, {arriveTangent=-1, interpMode=RCIM_Cubic, leaveTangent=-1, time=0.2, value=1}, <at 1 0.5>]}}})
    )) {module=ScaleMeshSize001}
   &ScaleColor(
    Scale Alpha=&FloatFromCurve(FloatCurve=DataInterface() {propertyValues {Curve {keys [<at 0 0 RCIM_Cubic>,<at 0.2 1 RCIM_Cubic>,<at 1 0 RCIM_Cubic>]}}}, Scale Curve=2)
    ScaleA=true
    ScaleRGB=true
   )
  ]
  rendererClasses [&NiagaraMeshRendererProperties]
  renderers [{rendererClass=&NiagaraMeshRendererProperties, rendererIndex=0}]
 }}
summary {
 systemName=NS_Emote_Floret_Tears
 emitters [{emitterName=Drips001, rendererClasses [&NiagaraMeshRendererProperties]}]
 userVariables [
  User.bIsLargeBodyPawn {type=NiagaraBool}
  User.SkeletalMesh {
   type=NiagaraDataInterfaceSkeletalMesh
   defaultValue {
    struct=NiagaraExt_VariableValue_DataInterface
    value {dataInterface=&self:NiagaraDataInterfaceSkeletalMesh_0}
   }
  }
 ]
}
